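# Niagara idle effect for the DareBergAdmit cosmetic shoes, adding frost sprites and ground decals at the feet.
# Frost spawns around foot bones in four partitions as flipbook sprites tinted by time of day, aging faster when moving.
# Decal bursts a pair of decals gated by owner speed and rotation, with color driven by FreezeLerp.
# Per-shoe scale parameters size or cull each side, and front-end modifiers adjust the look.
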NiagaraSystem /CosmeticShoes/Assets/FORT_Shoes/Shoe_DareBergAdmit/FX/NS_Shoe_DareBergAdmit_Idle
dependencies {
 usedDataInterfaces [
  &NiagaraDataInterfaceArrayFloat3
  &NiagaraDataInterfaceArrayFloat
  &NiagaraDataInterfaceCurve
  &NiagaraDataInterfaceSpriteRendererInfo
  /Script/Niagara.NiagaraDataInterfaceArrayQuat
  &NiagaraDataInterfaceSkeletalMesh
 ]
 usedDynamicInputs [&Multiply_Vector]
 usedModules [
  &self:SystemSpawnScript.SystemScriptSource.SystemScriptGraph.NiagaraNodeAssignment_0.SetVariables_#1
  &SystemState
  &self:SystemSpawnScript.SystemScriptSource.SystemScriptGraph.NiagaraNodeAssignment_1.SetVariables_#2
  &self:SystemSpawnScript.SystemScriptSource.SystemScriptGraph.NiagaraNodeAssignment_2.SetVariables_#3
 ]
}
emitters {
 Decal {
  emitterSpawnScript [/SetVariables(Emitter.DecalColor:LinearColor=<rgba 0.5703 0.8572 1 1>) {isSetParameters=true}]
  emitterUpdateScript [
   &self:HasRotatedBurstTrigger() {module=HasRotatedBurstTrigger}
   &EmitterState()
   &Timeline(Play=$Emitter.HasRotated, Play Rate=10)
   /SetVariables(Emitter.HasUpdatedLastFrame:NiagaraBool=$Emitter.HasRotated, Emitter.PreviousRotation:Quat4f=$Engine.Owner.Rotation) {isSetParameters=true}
   /Niagara/Modules/Emitter/SpawnByTrigger(
    Spawn Count=2
    Trigger:NiagaraBool=&Boolean_AndOperation(
     Bool1=&Boolean_AndOperation(
      Bool1=&SetBoolByFloatComparison(A=$Emitter.Timeline.Time, B=0.1, Comparison Type=::NewEnumerator6)
      Bool2=&SetBoolByFloatComparison(A=20, B=&VectorLength(Vector=$Engine.Owner.Velocity))
     )
     Bool2=&SetBoolByFloatComparison(A=$System.Timeline.Time, B=0.01, Comparison Type=::NewEnumerator6)
    )
    Trigger Count=1
    Trigger Interval=2
    TriggerCountSwitch=true
    TriggerIntervalSwitch=true
   )
  ]
  particleSpawnScript [
   &PartitionParticles()
   &InitializeParticle(Lifetime=2, UsePositionOffset=false, Write Material Random=true)
   /SetVariables(
    Particles.DecalSize:Vector3f=&FN_FrontEndVector3Modifier(Gameplay=<35 66 66>, Front End=<35 66 66>)
    Particles.Quat:Quat4f=/Niagara/DynamicInputs/Arrays/SelectQuaternionFromArray(
     Array Sampling Mode:ENiagaraArraySamplingMode=::NewEnumerator1
     Direct Array Index:NiagaraInt32=$Particles.PartitionParticles.Partition
     Interpolation:NiagaraFloat=?
     Direct Array Mode:ENiagara_WrapClamp=::NewEnumerator0
     Evaluation Type:ENiagaraRandomnessEvaluation=::NewEnumerator0
     Normalized Interpolation:NiagaraBool=true
     Recalculate Random Each Loop:NiagaraBool=false
     Randomness Mode:ENiagaraRandomnessMode=::NewEnumerator0
     Random Seed:NiagaraInt32=?
     Fixed Random Seed:NiagaraBool=false
     Quaternion Selection Array:NiagaraDataInterfaceArrayQuat=DataInterface() {propertyValues {GpuSyncMode=SyncCpuToGpu, MaxElements=0, QuatData [<0 -0.0087 0 1>, <0 -0.0087 0 1>]}}
     Enable Spherical Linear Interpolation:NiagaraBool=?
    )
   ) {isSetParameters=true}
   &SkeletalMeshLocation(
    Bone / Socket Index=&Multiply_Int(A=&ModuloInt(A=$Particles.PartitionParticles.Partition, B=2), B=2)
    Bone Sampling Mode=::NewEnumerator4
    MaskSpawnGroup=false
    Skeletal Mesh=$System.Skeletal Mesh Reader
   )
   /SetVariables() {isSetParameters=true}
   &KillParticles(Kill Particles=&SetBoolByIntComparison(A=$Engine.Emitter.NumParticles, B=16, Comparison Type=::NewEnumerator0))
   /SetVariables(Particles.DecalSize:Vector3f=&Multiply_VectorByFloat(
     Vector=$Particles.DecalSize
     Float=&MergeHelper_FloatSwitch(
      0=&Multiply_Float(A=&MakeFloatFromVector(Vector=$User.ShoeScale_L, Channel=::NewEnumerator0), B=1)
      1=&Multiply_Float(A=&MakeFloatFromVector(Vector=$User.ShoeScale_R, Channel=::NewEnumerator0), B=1)
      2=?
      3=?
      Index=$Particles.PartitionParticles.Partition
      Num=::NewEnumerator0
     )
    )) {isSetParameters=true}
   &KillParticles(Kill Particles=&SetBoolByFloatComparison(
     A=&DistanceBetweenPositions(A=/Niagara/DynamicInputs/Helpers/SimulationPosition(), B=$Particles.Position)
     B=10
     Comparison Type=::NewEnumerator6
    )) {module=KillParticles001}
  ]
  particleUpdateScript [
   &self:ScratchModule_02(HasTriggeredResume:NiagaraBool=$Particles.HasTriggeredResume, SpeedThreshold=22) {module=ScratchModule_02}
   &ParticleState(DeltaTime:NiagaraFloat=$Particles.CustomDeltaTime) {module=ParticleState001}
   &ScaleColor(
    Scale RGB=&MakeVector(
     X=&Add_Float(
      A=&FloatFromCurve(FloatCurve=DataInterface() {propertyValues {Curve {
          defaultValue=1
          keys [
           {interpMode=RCIM_Cubic, tangentMode=RCTM_User, time=0, value=1}
           {interpMode=RCIM_Cubic, tangentMode=RCTM_User, time=0.5, value=0}
           {tangentMode=RCTM_User, time=1, value=1}
          ]
         }}})
      B=&Lerp_Float(A=1, B=0, Alpha=&FN_FrontEndFloatModifier(Gameplay=$User.FreezeLerp, Front End=0))
     )
     Y=0
     Z=0
    )
    ScaleA=true
    ScaleRGB=true
   )
  ]
  rendererClasses [&NiagaraDecalRendererProperties]
  renderers [{rendererClass=&NiagaraDecalRendererProperties, rendererIndex=0}]
 }
 Frost {
  emitterUpdateScript [
   &EmitterState()
   &SpawnRate(
    Spawn Probability=&MakeCustomFloatFromBool(
     Bool=&SetBoolByFloatComparison(A=$System.Timeline.Time, B=0.1, Comparison Type=::NewEnumerator6)
     True Float=1
     False Float=0
    )
    SpawnRate=16
    Use Spawn Probability=true
   )
  ]
  particleSpawnScript [
   &PartitionParticles(Number Of Partitions=4)
   &InitializeParticle(
    Color=&FNTimeOfDayColor(
     Night Color=<rgba 0.2066 0.3684 0.776 0.075>
     Morning Color=<rgba 0.4323 0.6698 1 0.1>
     Day Color=<rgba 0.4323 0.6698 1 0.1>
     Evening Color=<rgba 0.4323 0.6698 1 0.05>
    )
    Lifetime=&MakeCustomFloatFromBool(
     Bool=$System.isMoving
     True Float=&RandomRangeFloat(Minimum=1, Maximum=0.5)
     False Float=&RandomRangeFloat(Minimum=2, Maximum=1.5)
    )
    Sprite Rotation Mode=::NewEnumerator2
    Sprite Size Mode=::NewEnumerator3
    Uniform Sprite Size=&Multiply_Float(
     A=&RandomRangeFloat(Minimum=30, Maximum=25)
     B=&FN_FrontEndFloatModifier(Gameplay=1, Front End=0.8)
    )
    UsePositionOffset=false
    Write Material Random=true
   )
   &CameraOffset(Camera Offset Amount=&MakeCustomFloatFromBool(Bool=$User.bIsOffPawnPreview, True Float=20, False Float=10))
   &SkeletalMeshLocation(
    Bone / Socket Index=$Particles.PartitionParticles.Partition
    Bone Sampling Mode=::NewEnumerator4
    MaskSpawnGroup=false
    Sampled Position Offset=&MergeHelper_VectorSwitch(
     0=&SelectVectorFromArray(
      Array Sampling Mode=::NewEnumerator0
      Direct Array Index=?
      Vector Selection Array=DataInterface() {propertyValues {FloatData [<-5 0 -3>, <-5 0 -3>, <-5 -7 0>]}}
      Random Seed=0
     )
     1=&SelectVectorFromArray(
      Array Sampling Mode=::NewEnumerator0
      Direct Array Index=?
      Vector Selection Array=DataInterface() {propertyValues {FloatData [<7 0 0>, <-4 3 -4>, <-4 3 4>]}}
      Random Seed=0
     )
     2=&SelectVectorFromArray(
      Array Sampling Mode=::NewEnumerator0
      Direct Array Index=?
      Vector Selection Array=DataInterface() {propertyValues {FloatData [<5 0 -3>, <5 0 3>, <5 7 0>]}}
      Random Seed=0
     )
     3=&SelectVectorFromArray(
      Array Sampling Mode=::NewEnumerator0
      Direct Array Index=?
      Vector Selection Array=DataInterface() {propertyValues {FloatData [<-7 0 0>, <4 3 -4>, <4 3 4>]}}
      Random Seed=0
     )
     Index=$Particles.PartitionParticles.Partition
     Num=::NewEnumerator2
    )
    Skeletal Mesh=$System.Skeletal Mesh Reader
   )
   &KillParticles(Kill Particles=&Boolean_OrOperation(
     Bool1=&Boolean_AndOperation(
      Bool1=&SetBoolByFloatComparison(
       A=&MakeFloatFromVector(Vector=$User.ShoeScale_L, Channel=::NewEnumerator0)
       B=0.5
       Comparison Type=::NewEnumerator6
      )
      Bool2=&SetBoolByIntComparison(A=$Particles.PartitionParticles.Partition, B=1, Comparison Type=::NewEnumerator7)
     )
     Bool2=&Boolean_AndOperation(
      Bool1=&SetBoolByFloatComparison(
       A=&MakeFloatFromVector(Vector=$User.ShoeScale_R, Channel=::NewEnumerator0)
       B=0.5
       Comparison Type=::NewEnumerator6
      )
      Bool2=&SetBoolByIntComparison(A=$Particles.PartitionParticles.Partition, B=2, Comparison Type=::NewEnumerator2)
     )
    ))
  ]
  particleUpdateScript [
   &ParticleState(DeltaTime:NiagaraFloat=&MakeCustomFloatFromBool(Bool=$System.isMoving, True Float=&Multiply_Float(A=$Engine.DeltaTime, B=2), False Float=$Engine.DeltaTime))
   &SocketLocation(
    Position Sampling=::NewEnumerator1
    SamplePosOffset=false
    Socket Index=1
    Socket Reader=$System.Socket Reader
    Socket Sampling Mode=::NewEnumerator3
    VelocitySpeedLimit=false
   ) {module=SocketLocation001}
   /SetVariables(Particles.isFeetMoving:NiagaraBool=&SetBoolByFloatComparison(A=20, B=&VectorLength(Vector=$Output.SocketLocation001.SampledVelocity))) {isSetParameters=true}
   &Timeline(
    Additional Output:ENiagara_TimelineOutput=::NewEnumerator1
    Additional Output Mode:ENiagara_TimelineOutputMode=::NewEnumerator1
    Float Curve=DataInterface() {propertyValues {Curve {defaultValue=0, keys [{tangentMode=RCTM_User, time=0, value=1}, {tangentMode=RCTM_User, time=1, value=0}]}}}
    Play=$Particles.isFeetMoving
    Rewind:NiagaraBool=&InvertBool(Boolean:NiagaraBool=$Particles.isFeetMoving)
    Rewind Mode:ENiagara_TimelineMode=::NewEnumerator1
    Rewind Rate=4
   )
   &ScaleSpriteSize(
    Scale Sprite Size Mode=::NewEnumerator0
    Uniform Scale Factor=&Multiply_Float(
     A=&MergeHelper_FloatSwitch(
      0=&Multiply_Float(A=&MakeFloatFromVector(Vector=$User.ShoeScale_L, Channel=::NewEnumerator0), B=1)
      1=&Multiply_Float(A=&MakeFloatFromVector(Vector=$User.ShoeScale_R, Channel=::NewEnumerator0), B=1)
      2=?
      3=?
      Index=/Niagara/DynamicInputs/Divide/Divide_Int(A:NiagaraInt32=$Particles.PartitionParticles.Partition, B:NiagaraInt32=2)
      Num=::NewEnumerator0
     )
     B=&FloatFromCurve(FloatCurve=DataInterface() {propertyValues {Curve {
         defaultValue=0
         keys [
          {arriveTangent=3.2023, interpMode=RCIM_Cubic, leaveTangent=3.2023, tangentMode=RCTM_User, time=0, value=0}
          {interpMode=RCIM_Cubic, tangentMode=RCTM_User, time=1, value=1}
         ]
        }}})
    )
   )
   &ScaleColor(
    Scale Alpha=&Multiply_Float(
     A=&MakeCustomFloatFromBool(Bool=$User.HideFootFX, True Float=0, False Float=1)
     B=&MaskFloatBySpawnGroup(
      Masked Float=&MakeFloatFromVector(Vector=$User.ShoeScale_L, Channel=::NewEnumerator0)
      Passthrough Float=&MakeFloatFromVector(Vector=$User.ShoeScale_R, Channel=::NewEnumerator0)
      Spawn Group Mask=$Particles.PartitionParticles.Partition
     )
    )
    ScaleA=true
    ScaleRGB=false
   ) {module=ScaleColor001}
   &SubUVAnimation(
    End Frame Range Override=10
    "Sprite Renderer "=DataInterface() {propertyValues {SpriteRenderer=&self:Frost_0.NiagaraSpriteRendererProperties_1}}
    UseEndFrame=true
    UseStartFrame=false
   )
   &ScaleColor(
    Scale Alpha=&FloatFromCurve(
     FloatCurve=DataInterface() {propertyValues {Curve {
        defaultValue=0
        keys [
         {interpMode=RCIM_Cubic, tangentMode=RCTM_User, time=0, value=0}
         {interpMode=RCIM_Cubic, tangentMode=RCTM_User, time=0.3, value=1}
         {tangentMode=RCTM_User, time=1, value=0}
        ]
       }}}
     Scale Curve=&Multiply_Float(
      A=&Multiply_Float(A=$System.Timeline002.Float, B=$System.Timeline001.Float)
      B=&FN_FrontEndFloatModifier(Gameplay=1, Front End=0.7)
     )
    )
    ScaleA=true
    ScaleRGB=false
   )
   &DynamicMaterialParameters(
    Index 0 Param 1=3
    Index 0 Param 2=1
    Index 0 Param 3=0.5
    Module.Index0Float0Write=true
    Module.Index0Float1Write=true
    Module.Index0Float2Write=true
    Module.Index0Float3Write=true
   )
   &InheritSourceMovement(
    Applied Position Scale=&Multiply_VectorByFloat(
     Vector=<1 1 1>
     Float=&FloatFromCurve(FloatCurve=DataInterface() {propertyValues {Curve {
         defaultValue=1
         keys [{interpMode=RCIM_Cubic, tangentMode=RCTM_User, time=0, value=1}, {interpMode=RCIM_Cubic, tangentMode=RCTM_User, time=1, value=0.1}]
        }}})
    )
    ApplyForce=false
    ApplyPosition=true
    ApplyPositionFalloff=false
   )
   &SolveForcesAndVelocity(Clamp Velocity=false, Limit Acceleration=false)
  ]
  rendererClasses [&NiagaraSpriteRendererProperties]
  renderers [{rendererClass=&NiagaraSpriteRendererProperties}]
 }
}
summary {
 systemName=NS_Shoe_DareBergAdmit_Idle
 emitters [
  {emitterName=Frost, rendererClasses [&NiagaraSpriteRendererProperties]}
  {emitterName=Decal, rendererClasses [&NiagaraDecalRendererProperties]}
 ]
 userVariables [
  User.bIsFrontEnd {type=NiagaraBool}
  User.bIsFrontEndPreview {type=NiagaraBool}
  User.bIsOffPawnPreview {type=NiagaraBool}
  User.FreezeLerp {type=NiagaraFloat, defaultValue=0, description=""}
  User.HideFootFX {type=NiagaraBool}
  User.isJumping {type=NiagaraBool}
  User.IsOwnerRotating {type=NiagaraBool}
  User.OwnerLocalVelocity {type=Vector3f, description="", defaultValue {struct=Vector3f, value=<1 1 1>}}
  User.ShoeScale_L {type=Vector3f, description="", defaultValue {struct=Vector3f, value=<1 1 1>}}
  User.ShoeScale_R {type=Vector3f, description="", defaultValue {struct=Vector3f, value=<1 1 1>}}
 ]
}
