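# Masked parent material for the DareBergAdmit cosmetic shoe, adding height-driven frost and ice over the 2023 character parent.
# A noise-eroded pre-skinned height gradient, scaled by FreezeIntensity or a looping preview and shaped by normals and red vertex color, forms the ice mask.
# It drives the ice function and speckled snow blend, and reveals red-painted geometry through opacity.
# Reactivity, sock, PostFX and hologram functions follow, then a masked specular clamp.
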Material /CosmeticShoes/Assets/FORT_Shoes/Shoe_DareBergAdmit/Materials/M_DareBergAdmit_Parent
graph {
 properties {
  bUseMaterialAttributes=true
  bUsedWithMorphTargets=true
  bUsedWithSkeletalMesh=true
  blendMode=BLEND_Masked
  pixelDepthOffsetMode=PDOM_Legacy
  previewMesh=/CosmeticShoes/Assets/FORT_Shoes/Shoe_DareBergAdmit/Meshes/DareBergAdmit_Shoe
  shadingModel=MSM_FromMaterialExpression
  usedShadingModels=Default Lit | Subsurface | Clear Coat | Subsurface Profile | Cloth
 }
 nodes [
  n1 = MaterialFunctionCall() {materialFunction=&MF_CharacterParent_2023_PreFX}
  n2 = LocalPosition() {localOrigin=InstancePreSkinning}
  n3 = Subtract(A=ComponentMask(None=VertexInterpolator(VS=@n2.XYZ).PS) {b=true}) {constB=-1}
  n4 = Time()
  n5 = Constant() {r=7}
  n6 = ConstantBiasScale(None=Frac(None=Divide(A=@n4, B=@n5))) {bias=-0.5, scale=2}
  n7 = Multiply(
   A=ScalarParameter() {defaultValue=14, parameterName=FrostHeight}
   B=LinearInterpolate(A=ScalarParameter() {defaultValue=1, parameterName=FreezeIntensity}, B=Abs(None=@n6), Alpha=ScalarParameter() {parameterName=LoopPreview})
  )
  n8 = OneMinus(None=Saturate(None=Divide(A=@n3, B=@n7)))
  n9 = TextureSample(UVs=Multiply(A=TextureCoordinate(), B=ScalarParameter() {defaultValue=2.5, group=FrostChill, parameterName=FrostTile})) {
   samplerSource=SSM_Wrap_WorldGroupSettings
   samplerType=SAMPLERTYPE_Masks
   texture=/CosmeticShoes/Assets/FORT_Shoes/Shoe_FrostChill/Textures/T_FrostNoise
  }
  n10 = Saturate(None=Subtract(A=Multiply(A=@n8) {constB=2}, B=OneMinus(None=@n9.R)))
  n11 = VertexInterpolator(VS=PreSkinnedNormal())
  n12 = OneMinus(None=Saturate(None=ComponentMask(None=@n11.PS) {b=true}))
  n13 = Ceil(None=VertexColor().R)
  n14 = Multiply(A=Saturate(None=Multiply(A=@n10) {constB=2}), B=Saturate(None=Add(A=@n12, B=@n13)))
  n15 = SetMaterialAttributes!(
   MaterialAttributes=@n1.Result
   Base Color=Constant3Vector() {constant=<rgba 0.9 0.9765 1 0>}
   Metallic=Constant()
   Specular=Constant() {r=0.3}
   Roughness=Constant() {r=0.6}
  ) {attributeSetTypes [
    #1
    #2
    #3
    #4
   ]}
  n16 = TextureSample(UVs=Multiply(A=TextureCoordinate(), B=ScalarParameter() {defaultValue=4, group=FrostChill, parameterName=SnowTile})) {samplerType=SAMPLERTYPE_Masks, texture=/CosmeticShoes/Assets/FORT_Shoes/Shoe_DareBergAdmit/Textures/T_DareBergAdmit_SnowSpecs}
  n17 = LinearInterpolate(Alpha=Ceil(None=VertexColor().R)) {constA=1, constB=0.15}
  n18 = BlendMaterialAttributes(
   A=MaterialFunctionCall!(
    HighlightMask=Constant()
    In=@n1.Result
    InMask=@n1.Result
    InFXMask=NamedRerouteDeclaration(None=@n14) {name=IceMask, nodeColor=<rgba 0 1 0.2734 1>}
   ) {materialFunction=&/MF_DareBergAdmit_Ice}.Result
   B=@n15
   Alpha=Multiply(A=@n14, B=Multiply(A=@n16.R, B=@n17))
  )
  n19 = LinearInterpolate(B=NamedRerouteUsage(), Alpha=Ceil(None=VertexColor().R)) {constA=1}
  n20 = MaterialFunctionCall!(
   InOpacity
   Material=MaterialFunctionCall!(Material=SetMaterialAttributes!(MaterialAttributes=@n18, Opacity Mask=@n19) {attributeSetTypes [#5]}) {materialFunction=/Game/Global/Materials/Cosmetics/Kicks/Shoes/Materials/MF_ShoeReactivityTest}.Result
   Mask=@n1.Tech Art Mask
  ) {materialFunction=/Game/Global/Materials/Cosmetics/Kicks/Socks/Materials/MF_Sock}
  n21 = GetMaterialAttributes(None=MaterialFunctionCall!(In=MaterialFunctionCall(Material=@n20.Result, M=@n1.M, Pre Skinned Local Position=@n1.M, Distance Blend=@n1.M, Fresnel=@n1.M) {materialFunction=&MF_CharacterParent_2023_PostFX}.Result) {materialFunction=/Game/Global/Materials/Cosmetics/Kicks/Shoes/Materials/MF_TechartShoeHologram}.Result) {attributeGetTypes [#3, #6]}
  n22 = ScalarParameter() {defaultValue=1, group=04_Tech Art, parameterName=Specular Clamp}
  SetMaterialAttributes!(
   MaterialAttributes=@n21.MaterialAttributes
   Specular=LinearInterpolate(
    A=@n21.Specular
    B=Clamp(None=@n21.Specular, Max=@n22)
    Alpha=ChannelMaskParameter(None=@n1.Tech Art Mask) {group=05_Tech Art Masks, parameterName=ClothFuzzMaskChannel}.R
   )
   Anisotropy
  ) {attributeSetTypes [#3, #6]}
 ]
}
